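# Physics asset for the crouching variant of the male medium AccentWall jacket, giving collision and joint limits to its dyn_jacket bones.
# Capsules cover the pelvis, spine, legs and left upper arm, linked by free angular constraints, while a huge root capsule sits low like a ground surface.
# Five jacket chains use per-segment swing limits and zero twist, mostly anchored to spine bones; the A-right chain has no parent constraint.
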PhysicsAsset /BRCosmetics/Characters/Player/Male/Medium/Bodies/M_MED_AccentWall/Meshes/M_MED_AccentWall_jacket_Crouching_Physics
bodies {
 calf_l [{shapeType=Capsule, length=34, radius=11.5, shapeName=calf_l_capsule, center=<-22.9453 0 0>, rotation=<rot 90 0 0>}]
 calf_r [{shapeType=Capsule, length=36.5, radius=9, shapeName=calf_l_capsule, center=<22.9453 0 0>, rotation=<rot 90 0 180>}]
 dyn_jacketA_l_01 [{shapeType=Capsule, length=8, radius=9, shapeName=dyn_jacketA_l_01_capsule, center=<8 -8 0>}]
 dyn_jacketA_l_02 [{shapeType=Capsule, length=8, radius=9, shapeName=dyn_jacketA_l_02_capsule, center=<8 -8 0>}]
 dyn_jacketA_l_03 [{shapeType=Capsule, length=8.5, radius=8.5, shapeName=dyn_jacketA_l_03_capsule, center=<8 -8 0>, rotation=<rot 0 0 7.343>}]
 dyn_jacketA_l_04 [{shapeType=Capsule, length=8, radius=9, shapeName=dyn_jacketA_l_04_capsule, center=<8 -8 0>, rotation=<rot 0 0.0001 11.4535>}]
 dyn_jacketA_l_05 [{shapeType=Capsule, length=8, radius=9, shapeName=dyn_jacketA_l_05_capsule, center=<8 -8 0>, rotation=<rot 0 0.0001 11.5723>}]
 dyn_jacketA_l_MSTR [{shapeType=Capsule, length=8, radius=9, shapeName=dyn_jacketA_l_MSTR_capsule, center=<8 -8 0>}]
 dyn_jacketA_r_01 [{shapeType=Capsule, length=20, radius=7.5, shapeName=dyn_jacketA_r_01_capsule, center=<4.9696 -6 0>}]
 dyn_jacketA_r_02 [{shapeType=Capsule, length=19.5, radius=8, shapeName=dyn_jacketA_r_02_capsule, center=<4.9696 -6 4>}]
 dyn_jacketA_r_03 [{shapeType=Capsule, length=19.5, radius=8, shapeName=dyn_jacketA_r_03_capsule, center=<4.9696 -6 4>, rotation=<rot 0 0 1.3674>}]
 dyn_jacketA_r_04 [{shapeType=Capsule, length=16.5, radius=11, shapeName=dyn_jacketA_r_04_capsule, center=<4.9696 -6 4>, rotation=<rot -0.0002 -0.0001 -4.1748>}]
 dyn_jacketA_r_05 [{shapeType=Capsule, length=16.5, radius=11, shapeName=dyn_jacketA_r_05_capsule, center=<4.9696 -6 4>, rotation=<rot -0.0002 -0.0001 -10.2531>}]
 dyn_jacketA_r_06 [{shapeType=Capsule, length=17, radius=10.5, shapeName=dyn_jacketA_r_06_capsule, center=<4.9696 -6 4>, rotation=<rot -0.0001 -0.0001 -9.0448>}]
 dyn_jacketB_l_01 [{shapeType=Capsule, length=0.5, radius=0.5, shapeName=dyn_jacketB_l_01_capsule, center=<5.0168 -6 0>}]
 dyn_jacketB_l_02 [{shapeType=Capsule, length=12, radius=7, shapeName=dyn_jacketB_l_02_capsule, center=<6.3385 -6 0>, rotation=<rot 0 0 10.2141>}]
 dyn_jacketB_l_03 [{shapeType=Capsule, length=12, radius=7, shapeName=dyn_jacketB_l_03_capsule, center=<6.3385 -6 0>}]
 dyn_jacketB_l_04 [{shapeType=Capsule, length=19.5, radius=8, shapeName=dyn_jacketB_l_04_capsule, center=<4.9696 -6 0>}]
 dyn_jacketB_l_05 [{shapeType=Capsule, length=11.5, radius=8.5, shapeName=dyn_jacketB_l_05_capsule, center=<4.9696 -6 0>}]
 dyn_jacketB_l_06 [{shapeType=Capsule, length=11, radius=9, shapeName=dyn_jacketB_l_06_capsule, center=<4.9696 -6 0>}]
 dyn_jacketB_l_MSTR [{shapeType=Capsule, length=0.5, radius=0.5, shapeName=dyn_jacketB_l_MSTR_capsule}]
 dyn_jacketB_r_01 [{shapeType=Capsule, length=20, radius=7.5, shapeName=dyn_jacketB_r_01_capsule, center=<4.9696 -6 0>}]
 dyn_jacketB_r_02 [{shapeType=Capsule, length=20, radius=7.5, shapeName=dyn_jacketB_r_02_capsule, center=<4.9696 -6 0>, rotation=<rot 0 0 -3.7567>}]
 dyn_jacketB_r_03 [{shapeType=Capsule, length=20, radius=7.5, shapeName=dyn_jacketB_r_03_capsule, center=<4.9696 -6 0>}]
 dyn_jacketB_r_04 [{shapeType=Capsule, length=19.5, radius=8, shapeName=dyn_jacketB_r_04_capsule, center=<4.9696 -6 0>}]
 dyn_jacketB_r_05 [{shapeType=Capsule, length=19.5, radius=8, shapeName=dyn_jacketB_r_05_capsule, center=<4.9696 -6 0>}]
 dyn_jacketB_r_MSTR [{shapeType=Capsule, length=0.5, radius=0.5, shapeName=dyn_jacketB_r_MSTR_capsule}]
 dyn_jacketC_r_02 [{shapeType=Capsule, length=21, radius=6.5, shapeName=dyn_jacketC_r_02_capsule, center=<4.9696 -6 0>}]
 dyn_jacketC_r_03 [{shapeType=Capsule, length=19.5, radius=8, shapeName=dyn_jacketC_r_03_capsule, center=<4.9696 -6 0>}]
 dyn_jacketC_r_04 [{shapeType=Capsule, length=17.5, radius=9.5, shapeName=dyn_jacketC_r_04_capsule, center=<4.9696 -8 0>}]
 dyn_jacketC_r_05 [{shapeType=Capsule, length=18, radius=9.5, shapeName=dyn_jacketC_r_05_capsule, center=<4.9696 -6 0>}]
 dyn_jacketC_r_MSTR [{shapeType=Capsule, length=0.5, radius=0.5, shapeName=dyn_jacketC_r_MSTR_capsule}]
 pelvis [{shapeType=Capsule, length=4, radius=16, shapeName=pelvis_capsule, center=<-10.6358 1.3673 0>}]
 root [{shapeType=Capsule, length=0, radius=300, shapeName=root_capsule, center=<0 0 -293.1135>, rotation=<rot 0 -90 0>}]
 spine_01 [{shapeType=Capsule, length=8, radius=12.5, shapeName=spine_01_capsule}]
 spine_03 [{shapeType=Capsule, length=10.5, radius=11.5, shapeName=spine_03_capsule, rotation=<rot 0 176.4237 169.9344>}]
 spine_04 [{shapeType=Capsule, length=13, radius=10.5, shapeName=spine_04_capsule}]
 spine_05 [{shapeType=Capsule, length=14.5, radius=9, shapeName=spine_05_capsule}]
 thigh_l [{shapeType=Capsule, length=36.5, radius=11.5, shapeName=thigh_l_capsule, center=<-20.4826 0 0>, rotation=<rot 90 0 0>}]
 thigh_r [{
   shapeType=Capsule
   length=40
   radius=11
   shapeName=thigh_l_capsule
   center=<19.4679 -1.2135 -3.2036>
   rotation=<rot 88.7602 0.0005 -179.9994>
  }]
 upperarm_l [{shapeType=Capsule, length=18.5, radius=6, shapeName=upperarm_l_capsule, center=<10.6427 0 0>, rotation=<rot 90 0 0>}]
}
constraints [
 {bone1Name=dyn_jacketB_r_01, bone2Name=dyn_jacketB_r_MSTR, swing1LimitDegrees=10, swing2LimitDegrees=20, twistLimitDegrees=0}
 {bone1Name=dyn_jacketB_r_02, bone2Name=dyn_jacketB_r_01, swing1LimitDegrees=45, swing2LimitDegrees=25, twistLimitDegrees=0}
 {bone1Name=dyn_jacketB_r_03, bone2Name=dyn_jacketB_r_02, swing1LimitDegrees=20, swing2LimitDegrees=25, twistLimitDegrees=0}
 {bone1Name=dyn_jacketB_r_04, bone2Name=dyn_jacketB_r_03, swing1LimitDegrees=20, swing2LimitDegrees=25, twistLimitDegrees=0}
 {bone1Name=dyn_jacketB_r_05, bone2Name=dyn_jacketB_r_04, swing1LimitDegrees=30, swing2LimitDegrees=10, twistLimitDegrees=0}
 {bone1Name=dyn_jacketA_r_02, bone2Name=dyn_jacketA_r_01, swing1LimitDegrees=25, swing2LimitDegrees=12, twistLimitDegrees=0}
 {bone1Name=dyn_jacketA_r_03, bone2Name=dyn_jacketA_r_02, swing1LimitDegrees=20, swing2LimitDegrees=20, twistLimitDegrees=0}
 {bone1Name=dyn_jacketA_r_04, bone2Name=dyn_jacketA_r_03, swing1LimitDegrees=20, swing2LimitDegrees=20, twistLimitDegrees=0}
 {bone1Name=dyn_jacketA_r_05, bone2Name=dyn_jacketA_r_04, swing1LimitDegrees=20, swing2LimitDegrees=20, twistLimitDegrees=0}
 {bone1Name=dyn_jacketA_r_06, bone2Name=dyn_jacketA_r_05, swing1LimitDegrees=30, swing2LimitDegrees=20, twistLimitDegrees=0}
 {bone1Name=dyn_jacketB_l_01, bone2Name=dyn_jacketB_l_MSTR, swing1LimitDegrees=30, swing2LimitDegrees=30, twistLimitDegrees=0}
 {bone1Name=dyn_jacketB_l_02, bone2Name=dyn_jacketB_l_01, swing1LimitDegrees=25, swing2LimitDegrees=5, twistLimitDegrees=0}
 {bone1Name=dyn_jacketB_l_03, bone2Name=dyn_jacketB_l_02, swing1LimitDegrees=25, swing2LimitDegrees=5, twistLimitDegrees=0}
 {bone1Name=dyn_jacketB_l_04, bone2Name=dyn_jacketB_l_03, swing1LimitDegrees=25, swing2LimitDegrees=5, twistLimitDegrees=0}
 {bone1Name=dyn_jacketB_l_05, bone2Name=dyn_jacketB_l_04, swing1LimitDegrees=20, swing2LimitDegrees=5, twistLimitDegrees=0}
 {bone1Name=dyn_jacketB_l_06, bone2Name=dyn_jacketB_l_05, swing1LimitDegrees=20, swing2LimitDegrees=5, twistLimitDegrees=0}
 {
  bone1Name=dyn_jacketB_l_MSTR
  bone2Name=spine_05
  swing1LimitDegrees=45
  swing1Motion=Free
  swing2LimitDegrees=45
  swing2Motion=Free
  twistLimitDegrees=45
  twistMotion=Free
 }
 {
  bone1Name=dyn_jacketB_r_MSTR
  bone2Name=spine_03
  swing1LimitDegrees=45
  swing1Motion=Free
  swing2LimitDegrees=45
  swing2Motion=Free
  twistLimitDegrees=45
  twistMotion=Free
 }
 {
  bone1Name=spine_01
  bone2Name=pelvis
  swing1LimitDegrees=45
  swing1Motion=Free
  swing2LimitDegrees=45
  swing2Motion=Free
  twistLimitDegrees=45
  twistMotion=Free
 }
 {
  bone1Name=spine_03
  bone2Name=spine_01
  swing1LimitDegrees=45
  swing1Motion=Free
  swing2LimitDegrees=45
  swing2Motion=Free
  twistLimitDegrees=45
  twistMotion=Free
 }
 {
  bone1Name=spine_04
  bone2Name=spine_03
  swing1LimitDegrees=45
  swing1Motion=Free
  swing2LimitDegrees=45
  swing2Motion=Free
  twistLimitDegrees=45
  twistMotion=Free
 }
 {
  bone1Name=thigh_l
  bone2Name=pelvis
  swing1LimitDegrees=45
  swing1Motion=Free
  swing2LimitDegrees=45
  swing2Motion=Free
  twistLimitDegrees=45
  twistMotion=Free
 }
 {
  bone1Name=calf_l
  bone2Name=thigh_l
  swing1LimitDegrees=45
  swing1Motion=Free
  swing2LimitDegrees=45
  swing2Motion=Free
  twistLimitDegrees=45
  twistMotion=Free
 }
 {
  bone1Name=spine_05
  bone2Name=spine_04
  swing1LimitDegrees=45
  swing1Motion=Free
  swing2LimitDegrees=45
  swing2Motion=Free
  twistLimitDegrees=45
  twistMotion=Free
 }
 {
  bone1Name=upperarm_l
  bone2Name=spine_05
  swing1LimitDegrees=45
  swing1Motion=Free
  swing2LimitDegrees=45
  swing2Motion=Free
  twistLimitDegrees=45
  twistMotion=Free
 }
 {
  bone1Name=thigh_r
  bone2Name=pelvis
  swing1LimitDegrees=45
  swing1Motion=Free
  swing2LimitDegrees=45
  swing2Motion=Free
  twistLimitDegrees=45
  twistMotion=Free
 }
 {
  bone1Name=calf_r
  bone2Name=thigh_r
  swing1LimitDegrees=45
  swing1Motion=Free
  swing2LimitDegrees=45
  swing2Motion=Free
  twistLimitDegrees=45
  twistMotion=Free
 }
 {
  bone1Name=pelvis
  bone2Name=root
  swing1LimitDegrees=45
  swing1Motion=Free
  swing2LimitDegrees=45
  swing2Motion=Free
  twistLimitDegrees=45
  twistMotion=Free
 }
 {bone1Name=dyn_jacketA_l_MSTR, bone2Name=spine_05, swing1LimitDegrees=20, swing2LimitDegrees=20, twistLimitDegrees=0}
 {bone1Name=dyn_jacketA_l_01, bone2Name=dyn_jacketA_l_MSTR, swing1LimitDegrees=10, swing2LimitDegrees=10, twistLimitDegrees=0}
 {bone1Name=dyn_jacketA_l_02, bone2Name=dyn_jacketA_l_01, swing1LimitDegrees=30, swing2LimitDegrees=10, twistLimitDegrees=0}
 {bone1Name=dyn_jacketA_l_03, bone2Name=dyn_jacketA_l_02, swing1LimitDegrees=30, swing2LimitDegrees=10, twistLimitDegrees=0}
 {bone1Name=dyn_jacketA_l_04, bone2Name=dyn_jacketA_l_03, swing1LimitDegrees=30, swing2LimitDegrees=10, twistLimitDegrees=0}
 {bone1Name=dyn_jacketA_l_05, bone2Name=dyn_jacketA_l_04, swing1LimitDegrees=30, swing2LimitDegrees=10, twistLimitDegrees=0}
 {
  bone1Name=dyn_jacketC_r_MSTR
  bone2Name=spine_01
  swing1LimitDegrees=45
  swing1Motion=Free
  swing2LimitDegrees=45
  swing2Motion=Free
  twistLimitDegrees=45
  twistMotion=Free
 }
 {bone1Name=dyn_jacketC_r_02, bone2Name=dyn_jacketC_r_MSTR, swing1LimitDegrees=45, swing2LimitDegrees=5, twistLimitDegrees=0}
 {bone1Name=dyn_jacketC_r_03, bone2Name=dyn_jacketC_r_02, swing1LimitDegrees=45, swing2LimitDegrees=5, twistLimitDegrees=0}
 {bone1Name=dyn_jacketC_r_04, bone2Name=dyn_jacketC_r_03, swing1LimitDegrees=45, swing2LimitDegrees=5, twistLimitDegrees=0}
 {bone1Name=dyn_jacketC_r_05, bone2Name=dyn_jacketC_r_04, swing1LimitDegrees=45, swing2LimitDegrees=5, twistLimitDegrees=0}
]
