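# Physics asset for the AccentWall male medium body's sleeve mesh, providing capsule bodies and joints for left-side jacket and sleeve bones plus one right jacket bone.
# Jacket A and sleeve chains hang from spine_05, the sleeve via a free master joint, with angular limits down each chain.
# Jacket B links only within itself.
# Pelvis, spine and left upper arm use free-rotation joints, above a huge floor-like root capsule.
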PhysicsAsset /BRCosmetics/Characters/Player/Male/Medium/Bodies/M_MED_AccentWall/Meshes/M_MED_AccentWall_sleeve_Physics
bodies {
 dyn_jacketA_l_01 [{shapeType=Capsule, length=17, radius=13, shapeName=dyn_jacketA_l_01_capsule, center=<0 11.9618 0>}]
 dyn_jacketA_l_02 [{shapeType=Capsule, length=2.5, radius=12.5, shapeName=dyn_jacketA_l_02_capsule, center=<0 13.052 -3.6111>}]
 dyn_jacketA_l_03 [{shapeType=Capsule, length=2, radius=13, shapeName=dyn_jacketA_l_03_capsule, center=<0 13.052 -3.6111>}]
 dyn_jacketB_l_02 [{shapeType=Capsule, length=7, radius=12, shapeName=dyn_jacketB_l_02_capsule, center=<0 14.5174 0>}]
 dyn_jacketB_l_03 [{shapeType=Capsule, length=6, radius=13, shapeName=dyn_jacketB_l_03_capsule, center=<0 13.052 0>}]
 dyn_jacketB_l_04 [{shapeType=Capsule, length=7, radius=12, shapeName=dyn_jacketB_l_04_capsule, center=<0 13.052 0>}]
 dyn_jacketC_r_02 [{shapeType=Capsule, length=13, radius=7.5, shapeName=dyn_jacketC_r_02_capsule, center=<7.4219 0 0>}]
 dyn_jacketSleeve_l_01 [{shapeType=Capsule, length=6.5, radius=6, shapeName=dyn_jacketSleeve_l_01_capsule, center=<6.4914 -5 0>}]
 dyn_jacketSleeve_l_02 [{shapeType=Capsule, length=6, radius=7, shapeName=dyn_jacketSleeve_l_02_capsule, center=<4.6537 -5 -1.2>}]
 dyn_jacketSleeve_l_03 [{shapeType=Capsule, length=5.5, radius=8, shapeName=dyn_jacketSleeve_l_03_capsule, center=<2.55 -5 -1.2>}]
 dyn_jacketSleeve_l_04 [{shapeType=Capsule, length=3.5, radius=10, shapeName=dyn_jacketSleeve_l_04_capsule, center=<2.55 -5 -1.2>}]
 dyn_jacketSleeve_l_MSTR [{shapeType=Capsule, length=0.5, radius=0.5, shapeName=dyn_jacketSleeve_l_MSTR_capsule}]
 pelvis [{shapeType=Capsule, length=4.5, radius=15.5, shapeName=pelvis_capsule, center=<-3.4034 2.2015 0>}]
 root [{shapeType=Capsule, length=0, radius=300, shapeName=root_capsule, center=<0 0 -293.1135>, rotation=<rot 0 -90 0>}]
 spine_01 [{shapeType=Capsule, length=9, radius=12, shapeName=spine_01_capsule}]
 spine_03 [{shapeType=Capsule, length=18.5, radius=13.5, shapeName=spine_03_capsule, center=<-1.0592 0 0>}]
 spine_04 [{shapeType=Capsule, length=19.5, radius=12.5, shapeName=spine_04_capsule, center=<0 0.0535 0>}]
 spine_05 [{shapeType=Capsule, length=18.5, radius=11, shapeName=spine_05_capsule, center=<0 -1.2713 0>}]
 upperarm_l [{shapeType=Capsule, length=18.5, radius=6, shapeName=upperarm_l_capsule, center=<10.6427 0 0>, rotation=<rot 90 0 0>}]
}
constraints [
 {bone1Name=dyn_jacketA_l_01, bone2Name=spine_05, swing1LimitDegrees=30, swing2LimitDegrees=30, twistLimitDegrees=30}
 {bone1Name=dyn_jacketA_l_02, bone2Name=dyn_jacketA_l_01, swing1LimitDegrees=30, swing2LimitDegrees=30, twistLimitDegrees=30}
 {bone1Name=dyn_jacketA_l_03, bone2Name=dyn_jacketA_l_02, swing1LimitDegrees=30, swing2LimitDegrees=30, twistLimitDegrees=30}
 {bone1Name=dyn_jacketB_l_03, bone2Name=dyn_jacketB_l_02, swing1LimitDegrees=20, swing2LimitDegrees=10, twistLimitDegrees=0}
 {bone1Name=dyn_jacketB_l_04, bone2Name=dyn_jacketB_l_03, swing1LimitDegrees=20, swing2LimitDegrees=10, twistLimitDegrees=0}
 {bone1Name=dyn_jacketSleeve_l_01, bone2Name=dyn_jacketSleeve_l_MSTR, swing1LimitDegrees=20, swing2LimitDegrees=6, twistLimitDegrees=0}
 {bone1Name=dyn_jacketSleeve_l_02, bone2Name=dyn_jacketSleeve_l_01, swing1LimitDegrees=15, swing2LimitDegrees=8, twistLimitDegrees=0}
 {bone1Name=dyn_jacketSleeve_l_03, bone2Name=dyn_jacketSleeve_l_02, swing1LimitDegrees=10, swing2LimitDegrees=10, twistLimitDegrees=0}
 {bone1Name=dyn_jacketSleeve_l_04, bone2Name=dyn_jacketSleeve_l_03, swing1LimitDegrees=10, swing2LimitDegrees=10, twistLimitDegrees=0}
 {
  bone1Name=dyn_jacketSleeve_l_MSTR
  bone2Name=spine_05
  swing1LimitDegrees=45
  swing1Motion=Free
  swing2LimitDegrees=45
  swing2Motion=Free
  twistLimitDegrees=45
  twistMotion=Free
 }
 {
  bone1Name=spine_01
  bone2Name=pelvis
  swing1LimitDegrees=45
  swing1Motion=Free
  swing2LimitDegrees=45
  swing2Motion=Free
  twistLimitDegrees=45
  twistMotion=Free
 }
 {
  bone1Name=spine_03
  bone2Name=spine_01
  swing1LimitDegrees=45
  swing1Motion=Free
  swing2LimitDegrees=45
  swing2Motion=Free
  twistLimitDegrees=45
  twistMotion=Free
 }
 {
  bone1Name=spine_04
  bone2Name=spine_03
  swing1LimitDegrees=45
  swing1Motion=Free
  swing2LimitDegrees=45
  swing2Motion=Free
  twistLimitDegrees=45
  twistMotion=Free
 }
 {
  bone1Name=spine_05
  bone2Name=spine_04
  swing1LimitDegrees=45
  swing1Motion=Free
  swing2LimitDegrees=45
  swing2Motion=Free
  twistLimitDegrees=45
  twistMotion=Free
 }
 {
  bone1Name=upperarm_l
  bone2Name=spine_05
  swing1LimitDegrees=45
  swing1Motion=Free
  swing2LimitDegrees=45
  swing2Motion=Free
  twistLimitDegrees=45
  twistMotion=Free
 }
 {
  bone1Name=pelvis
  bone2Name=root
  swing1LimitDegrees=45
  swing1Motion=Free
  swing2LimitDegrees=45
  swing2Motion=Free
  twistLimitDegrees=45
  twistMotion=Free
 }
 {
  bone1Name=dyn_jacketC_r_02
  bone2Name=spine_01
  swing1LimitDegrees=45
  swing1Motion=Free
  swing2LimitDegrees=45
  swing2Motion=Free
  twistLimitDegrees=45
  twistMotion=Free
 }
]
