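# Animation Blueprint for the AccentWall male body cosmetic, adding secondary motion to its jacket, seaweed and pants.
# It copies the attached parent's pose through a Control Rig and corrective bone offsets.
# RigidBody physics variants are chosen by backbling, crouch, front-end and long-cape state, followed by AnimDynamics chains.
# Cached poses link the stages, ending in parachute offsets and a per-bone DefaultSlot blend for custom idles.
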AnimBlueprint /BRCosmetics/Characters/Player/Male/Medium/Bodies/M_MED_AccentWall/Meshes/M_MED_AccentWall_AnimBP
target_skeleton=&/M_MED_AccentWall_Skeleton
graphs [
 AnimGraph {nodes [
   n1 = Layered blend per bone(
    BasePose=Use cached pose 'FEIdleCache'()
    BlendPoses_0=Slot 'DefaultSlot'()
    BlendWeights_0=Get CustomIdleIsPlayingFloat() -> (CustomIdleIsPlayingFloat:float)
   ) {layerSetup [{branchFilters [{boneName=dyn_claw_l_01}, {boneName=dyn_thumb_l_01}, {boneName=dyn_jacketCollar_l_01}]}]}
   Output Pose(Result=@n1) {layerGroup=None}
   n2 = Control Rig!(
    Source=Copy Pose From Mesh() {bUseAttachedParent=true}
    Alpha
    isCrouching:double=Get isCrouchingAlpha() -> (isCrouchingAlpha:double)
    isPlayingCustomEmote:bool=Get isPlayingCustomEmote() -> (isPlayingCustomEmote:bool)
   ) -> (Pose) {
    alphaBoolBlend {blendOption=Linear}
    controlRigAssetReference {blueprintRigClass=&/M_MED_AccentWall_CtrlRig.M_MED_AccentWall_CtrlRig_C}
   }
   Save cached pose 'CRCache'(Pose=@n2) {cacheName=CRCache}
   n3 = "Constraint - dyn_jacketB_l_MSTR (PP)"!(ComponentPose=Local To Component(LocalPose=Use cached pose 'CRCache'()), Alpha=1, ConstraintWeights_0=1, ConstraintWeights_1:float=0) -> (Pose) {
    alphaBoolBlend {blendOption=Linear}
    boneToModify {boneName=dyn_jacketB_l_MSTR}
    constraintSetup [{targetBone {boneName=spine_03}}, {targetBone {boneName=spine_04}}]
    constraintWeights [1,1]
   }
   n4 = "Transform (Modify) Bone - Bone: calf_r"(ComponentPose=@n3, Alpha=1, Translation="0,0,0", Rotation="0,-10,0", Scale="1,1,1") {
    rotationMode=BMM_Additive
    rotationSpace=BCS_BoneSpace
    alphaBoolBlend {blendOption=Linear}
    boneToModify {boneName=calf_r}
    rotation {yaw=-10}
   }
   n5 = "Transform (Modify) Bone - Bone: dyn_jacketC_r_MSTR"!(ComponentPose=@n4, bAlphaBoolEnabled:bool=Get bIsSliding(), Translation="0,-5,0", Rotation="0,0,0", Scale="1,1,1") -> (Pose) {
    alphaInputType=Bool
    rotationMode=BMM_Additive
    rotationSpace=BCS_BoneSpace
    translationMode=BMM_Additive
    translationSpace=BCS_BoneSpace
    alphaBoolBlend {blendOption=Linear}
    boneToModify {boneName=dyn_jacketC_r_MSTR}
    translation {y=-5}
   }
   n6 = Get CustomIdleIsPlayingFloat()
   n7 = "Transform (Modify) Bone - Bone: dyn_jacketA_l_MSTR"(ComponentPose=@n5, Alpha=@n6, Translation="0,-13,-3", Rotation="0,0,0", Scale="1,1,1") {
    rotationMode=BMM_Additive
    rotationSpace=BCS_BoneSpace
    translationMode=BMM_Additive
    translationSpace=BCS_BoneSpace
    alphaBoolBlend {blendOption=Linear}
    boneToModify {boneName=dyn_jacketA_l_MSTR}
    translation {y=-13, z=-3}
   }
   n8 = "Transform (Modify) Bone - Bone: dyn_jacketSleeve_l_MSTR"(ComponentPose=@n7, Alpha=@n6, Translation="0,-6,-2", Rotation="0,0,0", Scale="1,1,1") {
    rotationMode=BMM_Additive
    rotationSpace=BCS_BoneSpace
    translationMode=BMM_Additive
    translationSpace=BCS_BoneSpace
    alphaBoolBlend {blendOption=Linear}
    boneToModify {boneName=dyn_jacketSleeve_l_MSTR}
    translation {y=-6, z=-2}
   }
   n9 = Get notCustomIdlePlaying() -> (notCustomIdlePlaying:bool)
   n10 = "Transform (Modify) Bone - Bone: dyn_jacketA_l_MSTR"!(ComponentPose=@n8, bAlphaBoolEnabled:bool=@n9, Translation="0,-4,-3", Rotation="0,0,0", Scale="1,1,1") -> (Pose) {
    alphaInputType=Bool
    rotationMode=BMM_Additive
    rotationSpace=BCS_BoneSpace
    translationMode=BMM_Additive
    translationSpace=BCS_BoneSpace
    alphaBoolBlend {blendOption=Linear}
    boneToModify {boneName=dyn_jacketA_l_MSTR}
    translation {y=-4, z=-3}
   }
   n11 = "Transform (Modify) Bone - Bone: dyn_jacketSleeve_l_MSTR"!(ComponentPose=@n10, bAlphaBoolEnabled:bool=@n9, Translation="0,-1,-2", Rotation="0,0,0", Scale="1,1,1") -> (Pose) {
    alphaInputType=Bool
    rotationMode=BMM_Additive
    rotationSpace=BCS_BoneSpace
    translationMode=BMM_Additive
    translationSpace=BCS_BoneSpace
    alphaBoolBlend {blendOption=Linear}
    boneToModify {boneName=dyn_jacketSleeve_l_MSTR}
    translation {y=-1, z=-2}
   }
   Save cached pose 'preSimCache'(Pose=Component To Local(ComponentPose=@n11)) {cacheName=preSimCache}
   n12 = "Constraint - dyn_shoulderSeeweed_r_MSTR (PPPPPPPP)"!(
    ComponentPose=Local To Component(LocalPose=Use cached pose 'RBANCache'())
    Alpha=1
    ConstraintWeights_0=0.4927
    ConstraintWeights_1:float=0.1757
    ConstraintWeights_2:float=0.1617
    ConstraintWeights_3:float=0.0874
    ConstraintWeights_4:float=0.0575
    ConstraintWeights_5:float=0.0088
    ConstraintWeights_6:float=0.0082
    ConstraintWeights_7:float=0.008
   ) -> (Pose) {
    alphaBoolBlend {blendOption=Linear}
    boneToModify {boneName=dyn_shoulderSeeweed_r_MSTR}
    constraintSetup [
     {targetBone {boneName=upperarm_twist_01_r}}
     {targetBone {boneName=deform_upperarm_up_r}}
     {targetBone {boneName=upperarm_twist_02_r}}
     {targetBone {boneName=clavicle_r}}
     {targetBone {boneName=upperarm_r}}
     {targetBone {boneName=deform_upperarm_fr_r}}
     {targetBone {boneName=deform_upperarm_bk_r}}
     {targetBone {boneName=spine_03}}
    ]
    constraintWeights [1,1,1,1,1,1,1,1]
   }
   n13 = Anim Dynamics - Bone: dyn_shoulderSeeweed_r_01(ComponentPose=@n12, Alpha=1, ExternalForce="0,0,0") {
    angularDampingOverride=1
    angularSpringConstant=5000
    bAngularSpring=true
    bOverrideAngularDamping=true
    bUseGravityOverride=true
    gravityScale=0.35
    simulationSpace=Actor
    alphaBoolBlend {blendOption=Linear}
    boundBone {boneName=dyn_shoulderSeeweed_r_01}
    componentAppliedLinearAccClamp=<6000 6000 6000>
    componentLinearAccScale=<0.5 0.5 0.5>
    componentLinearVelScale=<0.5 0.5 0.5>
    gravityOverride {z=-980}
    physicsBodyDefinitions [{
      boundBone {boneName=dyn_shoulderSeeweed_r_01}
      boxExtents=<20 30 30>
      constraintSetup {angularLimitsMax=<15 15 0>, angularLimitsMin=<-15 -15 -30>, angularTarget=<1 0 0>}
      localJointOffset=<4 0 0>
     }]
    relativeSpaceBone {boneName=spine_05}
   }
   n14 = Anim Dynamics - Bone: dyn_shoulderSeeweed_r_02(ComponentPose=@n13, Alpha=1, ExternalForce="0,0,0") {
    angularDampingOverride=1
    angularSpringConstant=5000
    bAngularSpring=true
    bOverrideAngularDamping=true
    bUseGravityOverride=true
    gravityScale=0.35
    simulationSpace=Actor
    alphaBoolBlend {blendOption=Linear}
    boundBone {boneName=dyn_shoulderSeeweed_r_02}
    componentAppliedLinearAccClamp=<6000 6000 6000>
    componentLinearAccScale=<0.5 0.5 0.5>
    componentLinearVelScale=<0.5 0.5 0.5>
    gravityOverride {z=-980}
    physicsBodyDefinitions [{
      boundBone {boneName=dyn_shoulderSeeweed_r_02}
      boxExtents=<20 30 30>
      constraintSetup {angularLimitsMax=<15 15 -10>, angularLimitsMin=<-15 -15 -30>, angularTarget=<1 0 0>}
      localJointOffset=<4 0 0>
     }]
    relativeSpaceBone {boneName=spine_05}
   }
   n15 = Anim Dynamics - Bone: dyn_shoulderSeeweed_r_03(ComponentPose=@n14, Alpha=1, ExternalForce="0,0,0") {
    angularDampingOverride=1
    angularSpringConstant=5000
    bAngularSpring=true
    bOverrideAngularDamping=true
    bUseGravityOverride=true
    gravityScale=0.35
    simulationSpace=Actor
    alphaBoolBlend {blendOption=Linear}
    boundBone {boneName=dyn_shoulderSeeweed_r_03}
    componentAppliedLinearAccClamp=<6000 6000 6000>
    componentLinearAccScale=<0.5 0.5 0.5>
    componentLinearVelScale=<0.5 0.5 0.5>
    gravityOverride {z=-980}
    physicsBodyDefinitions [{
      boundBone {boneName=dyn_shoulderSeeweed_r_03}
      boxExtents=<20 30 30>
      constraintSetup {angularLimitsMax=<15 15 -10>, angularLimitsMin=<-15 -15 -30>, angularTarget=<1 0 0>}
      localJointOffset=<4 0 0>
     }]
    relativeSpaceBone {boneName=spine_05}
   }
   n16 = Anim Dynamics - Bone: dyn_pantPiecesA_01(ComponentPose=@n15, Alpha=1, ExternalForce="0,0,0") {
    angularDampingOverride=1
    angularSpringConstant=8000
    bAngularSpring=true
    bOverrideAngularDamping=true
    bUseGravityOverride=true
    gravityScale=0.35
    simulationSpace=RootRelative
    alphaBoolBlend {blendOption=Linear}
    boundBone {boneName=dyn_pantPiecesA_01}
    componentLinearAccScale=<1 1 1>
    componentLinearVelScale=<1 1 1>
    gravityOverride {z=-980}
    physicsBodyDefinitions [{
      boundBone {boneName=dyn_pantPiecesA_01}
      boxExtents=<20 30 30>
      constraintSetup {angularLimitsMax=<15 15 15>, angularLimitsMin=<-15 -15 -15>, angularTarget=<1 0 0>}
      localJointOffset=<4 0 0>
     }]
   }
   n17 = Anim Dynamics - Bone: dyn_pantPiecesB_01(ComponentPose=@n16, Alpha=1, ExternalForce="0,0,0") {
    angularDampingOverride=1
    angularSpringConstant=8000
    bAngularSpring=true
    bOverrideAngularDamping=true
    bUseGravityOverride=true
    gravityScale=0.35
    simulationSpace=RootRelative
    alphaBoolBlend {blendOption=Linear}
    boundBone {boneName=dyn_pantPiecesB_01}
    componentLinearAccScale=<1 1 1>
    componentLinearVelScale=<1 1 1>
    gravityOverride {z=-980}
    physicsBodyDefinitions [{
      boundBone {boneName=dyn_pantPiecesB_01}
      boxExtents=<20 30 30>
      constraintSetup {angularLimitsMax=<15 15 15>, angularLimitsMin=<-15 -15 -15>, angularTarget=<1 0 0>}
      localJointOffset=<4 0 0>
     }]
   }
   n18 = Anim Dynamics - Bone: dyn_pantPiecesC_01(ComponentPose=@n17, Alpha=1, ExternalForce="0,0,0") {
    angularDampingOverride=1
    angularSpringConstant=8000
    bAngularSpring=true
    bOverrideAngularDamping=true
    bUseGravityOverride=true
    gravityScale=0.35
    simulationSpace=RootRelative
    alphaBoolBlend {blendOption=Linear}
    boundBone {boneName=dyn_pantPiecesC_01}
    componentLinearAccScale=<1 1 1>
    componentLinearVelScale=<1 1 1>
    gravityOverride {z=-980}
    physicsBodyDefinitions [{
      boundBone {boneName=dyn_pantPiecesC_01}
      boxExtents=<20 30 30>
      constraintSetup {angularLimitsMax=<15 15 15>, angularLimitsMin=<-15 -15 -15>, angularTarget=<1 0 0>}
      localJointOffset=<4 0 0>
     }]
   }
   n19 = Anim Dynamics - Bone: dyn_legSeeweed_l_01(ComponentPose=@n18, Alpha=1, ExternalForce="0,0,0") {
    angularDampingOverride=1
    angularSpringConstant=8000
    bAngularSpring=true
    bOverrideAngularDamping=true
    bUseGravityOverride=true
    gravityScale=0.35
    simulationSpace=RootRelative
    alphaBoolBlend {blendOption=Linear}
    boundBone {boneName=dyn_legSeeweed_l_01}
    componentLinearAccScale=<1 1 1>
    componentLinearVelScale=<1 1 1>
    gravityOverride {z=-980}
    physicsBodyDefinitions [{
      boundBone {boneName=dyn_legSeeweed_l_01}
      boxExtents=<20 30 30>
      constraintSetup {angularLimitsMax=<15 15 15>, angularLimitsMin=<-15 -15 -15>, angularTarget=<1 0 0>}
      localJointOffset=<4 0 0>
     }]
   }
   n20 = Anim Dynamics - Bone: dyn_legSeeweed_l_02(ComponentPose=@n19, Alpha=1, ExternalForce="0,0,0") {
    angularDampingOverride=1
    angularSpringConstant=8000
    bAngularSpring=true
    bOverrideAngularDamping=true
    bUseGravityOverride=true
    gravityScale=0.35
    simulationSpace=RootRelative
    alphaBoolBlend {blendOption=Linear}
    boundBone {boneName=dyn_legSeeweed_l_02}
    componentLinearAccScale=<1 1 1>
    componentLinearVelScale=<1 1 1>
    gravityOverride {z=-980}
    physicsBodyDefinitions [{
      boundBone {boneName=dyn_legSeeweed_l_02}
      boxExtents=<20 30 30>
      constraintSetup {angularLimitsMax=<15 15 15>, angularLimitsMin=<-15 -15 -15>, angularTarget=<1 0 0>}
      localJointOffset=<4 0 0>
     }]
   }
   n21 = Anim Dynamics - Bone: dyn_legSeeweed_l_03(ComponentPose=@n20, Alpha=1, ExternalForce="0,0,0") {
    angularDampingOverride=1
    angularSpringConstant=8000
    bAngularSpring=true
    bOverrideAngularDamping=true
    bUseGravityOverride=true
    gravityScale=0.35
    simulationSpace=RootRelative
    alphaBoolBlend {blendOption=Linear}
    boundBone {boneName=dyn_legSeeweed_l_03}
    componentLinearAccScale=<1 1 1>
    componentLinearVelScale=<1 1 1>
    gravityOverride {z=-980}
    physicsBodyDefinitions [{
      boundBone {boneName=dyn_legSeeweed_l_03}
      boxExtents=<20 30 30>
      constraintSetup {angularLimitsMax=<15 15 15>, angularLimitsMin=<-15 -15 -15>, angularTarget=<1 0 0>}
      localJointOffset=<4 0 0>
     }]
   }
   n22 = Anim Dynamics - Bone: dyn_hipSeeweedA_l_01(ComponentPose=@n21, Alpha=1, ExternalForce="0,0,0") {
    angularDampingOverride=1
    angularSpringConstant=8000
    bAngularSpring=true
    bOverrideAngularDamping=true
    bUseGravityOverride=true
    gravityScale=0.35
    simulationSpace=RootRelative
    alphaBoolBlend {blendOption=Linear}
    boundBone {boneName=dyn_hipSeeweedA_l_01}
    componentLinearAccScale=<1 1 1>
    componentLinearVelScale=<1 1 1>
    gravityOverride {z=-980}
    physicsBodyDefinitions [{
      boundBone {boneName=dyn_hipSeeweedA_l_01}
      boxExtents=<20 30 30>
      constraintSetup {angularLimitsMax=<15 0 15>, angularLimitsMin=<-15 -15 -15>, angularTarget=<1 0 0>}
      localJointOffset=<4 0 0>
     }]
   }
   n23 = Anim Dynamics - Bone: dyn_hipSeeweedA_l_02(ComponentPose=@n22, Alpha=1, ExternalForce="0,0,0") {
    angularDampingOverride=1
    angularSpringConstant=8000
    bAngularSpring=true
    bOverrideAngularDamping=true
    bUseGravityOverride=true
    gravityScale=0.35
    simulationSpace=RootRelative
    alphaBoolBlend {blendOption=Linear}
    boundBone {boneName=dyn_hipSeeweedA_l_02}
    componentLinearAccScale=<1 1 1>
    componentLinearVelScale=<1 1 1>
    gravityOverride {z=-980}
    physicsBodyDefinitions [{
      boundBone {boneName=dyn_hipSeeweedA_l_02}
      boxExtents=<20 30 30>
      constraintSetup {angularLimitsMax=<15 0 15>, angularLimitsMin=<-15 -15 -15>, angularTarget=<1 0 0>}
      localJointOffset=<4 0 0>
     }]
   }
   n24 = Anim Dynamics - Bone: dyn_hipSeeweedA_l_03(ComponentPose=@n23, Alpha=1, ExternalForce="0,0,0") {
    angularDampingOverride=1
    angularSpringConstant=8000
    bAngularSpring=true
    bOverrideAngularDamping=true
    bUseGravityOverride=true
    gravityScale=0.35
    simulationSpace=RootRelative
    alphaBoolBlend {blendOption=Linear}
    boundBone {boneName=dyn_hipSeeweedA_l_03}
    componentLinearAccScale=<1 1 1>
    componentLinearVelScale=<1 1 1>
    gravityOverride {z=-980}
    physicsBodyDefinitions [{
      boundBone {boneName=dyn_hipSeeweedA_l_03}
      boxExtents=<20 30 30>
      constraintSetup {angularLimitsMax=<15 0 15>, angularLimitsMin=<-15 -15 -15>, angularTarget=<1 0 0>}
      localJointOffset=<4 0 0>
     }]
   }
   n25 = Anim Dynamics - Bone: dyn_hipSeeweedB_l_01(ComponentPose=@n24, Alpha=1, ExternalForce="0,0,0") {
    angularDampingOverride=1
    angularSpringConstant=8000
    bAngularSpring=true
    bOverrideAngularDamping=true
    bUseGravityOverride=true
    gravityScale=0.35
    simulationSpace=RootRelative
    alphaBoolBlend {blendOption=Linear}
    boundBone {boneName=dyn_hipSeeweedB_l_01}
    componentLinearAccScale=<1 1 1>
    componentLinearVelScale=<1 1 1>
    gravityOverride {z=-980}
    physicsBodyDefinitions [{
      boundBone {boneName=dyn_hipSeeweedB_l_01}
      boxExtents=<20 30 30>
      constraintSetup {angularLimitsMax=<15 0 15>, angularLimitsMin=<-15 -15 -15>, angularTarget=<1 0 0>}
      localJointOffset=<4 0 0>
     }]
   }
   n26 = Anim Dynamics - Bone: dyn_seeweedArm_l_01(ComponentPose=@n25, Alpha=1, ExternalForce="0,0,0") {
    angularDampingOverride=1
    angularSpringConstant=8000
    bAngularSpring=true
    bOverrideAngularDamping=true
    bUseGravityOverride=true
    gravityScale=0.35
    simulationSpace=World
    alphaBoolBlend {blendOption=Linear}
    boundBone {boneName=dyn_seeweedArm_l_01}
    componentLinearAccScale=<1 1 1>
    componentLinearVelScale=<1 1 1>
    gravityOverride {z=-980}
    physicsBodyDefinitions [{
      boundBone {boneName=dyn_seeweedArm_l_01}
      boxExtents=<20 30 30>
      constraintSetup {angularLimitsMax=<0 15 5>, angularLimitsMin=<0 -45 -5>, angularTarget=<1 0 0>}
      localJointOffset=<4 0 0>
     }]
    relativeSpaceBone {boneName=root}
   }
   n27 = Anim Dynamics - Bone: dyn_seeweedArm_l_02(ComponentPose=@n26, Alpha=1, ExternalForce="0,0,0") {
    angularDampingOverride=1
    angularSpringConstant=8000
    bAngularSpring=true
    bOverrideAngularDamping=true
    bUseGravityOverride=true
    gravityScale=0.35
    simulationSpace=World
    alphaBoolBlend {blendOption=Linear}
    boundBone {boneName=dyn_seeweedArm_l_02}
    componentLinearAccScale=<1 1 1>
    componentLinearVelScale=<1 1 1>
    gravityOverride {z=-980}
    physicsBodyDefinitions [{
      boundBone {boneName=dyn_seeweedArm_l_02}
      boxExtents=<20 30 30>
      constraintSetup {angularLimitsMax=<0 15 5>, angularLimitsMin=<0 -45 -5>, angularTarget=<1 0 0>}
      localJointOffset=<4 0 0>
     }]
    relativeSpaceBone {boneName=root}
   }
   n28 = Anim Dynamics - Bone: dyn_seeweedArm_l_03(ComponentPose=@n27, Alpha=1, ExternalForce="0,0,0") {
    angularDampingOverride=1
    angularSpringConstant=8000
    bAngularSpring=true
    bOverrideAngularDamping=true
    bUseGravityOverride=true
    gravityScale=0.35
    simulationSpace=World
    alphaBoolBlend {blendOption=Linear}
    boundBone {boneName=dyn_seeweedArm_l_03}
    componentLinearAccScale=<1 1 1>
    componentLinearVelScale=<1 1 1>
    gravityOverride {z=-980}
    physicsBodyDefinitions [{
      boundBone {boneName=dyn_seeweedArm_l_03}
      boxExtents=<20 30 30>
      constraintSetup {angularLimitsMax=<0 15 5>, angularLimitsMin=<0 -45 -5>, angularTarget=<1 0 0>}
      localJointOffset=<4 0 0>
     }]
    relativeSpaceBone {boneName=root}
   }
   n29 = Anim Dynamics - Bone: dyn_pant_l_01(ComponentPose=@n28, Alpha=1, ExternalForce="0,0,0") {
    angularDampingOverride=1
    angularSpringConstant=8000
    bAngularSpring=true
    bOverrideAngularDamping=true
    bUseGravityOverride=true
    gravityScale=0.35
    simulationSpace=RootRelative
    alphaBoolBlend {blendOption=Linear}
    boundBone {boneName=dyn_pant_l_01}
    componentLinearAccScale=<1 1 1>
    componentLinearVelScale=<1 1 1>
    gravityOverride {z=-980}
    physicsBodyDefinitions [{
      boundBone {boneName=dyn_pant_l_01}
      boxExtents=<20 30 30>
      constraintSetup {angularLimitsMax=<0 0 15>, angularLimitsMin=<0 -15 0>, angularTarget=<1 0 0>}
      localJointOffset=<4 0 0>
     }]
   }
   n30 = Anim Dynamics - Bone: dyn_pant_r_01(ComponentPose=@n29, Alpha=1, ExternalForce="0,0,0") {
    angularDampingOverride=1
    angularSpringConstant=8000
    bAngularSpring=true
    bOverrideAngularDamping=true
    bUseGravityOverride=true
    gravityScale=0.35
    simulationSpace=RootRelative
    alphaBoolBlend {blendOption=Linear}
    boundBone {boneName=dyn_pant_r_01}
    componentLinearAccScale=<1 1 1>
    componentLinearVelScale=<1 1 1>
    gravityOverride {z=-980}
    physicsBodyDefinitions [{
      boundBone {boneName=dyn_pant_r_01}
      boxExtents=<20 30 30>
      constraintSetup {angularLimitsMax=<0 0 0>, angularLimitsMin=<0 -15 0>, angularTarget=<1 0 0>}
      localJointOffset=<4 0 0>
     }]
   }
   Save cached pose 'dynCache'(Pose=Component To Local(ComponentPose=@n30)) {cacheName=dynCache}
   n31 = Get isParachutFloating() -> (isParachutFloating:bool)
   n32 = "Transform (Modify) Bone - Bone: dyn_jacketB_l_MSTR"!(
    ComponentPose=Local To Component(LocalPose=Use cached pose 'dynCache'())
    bAlphaBoolEnabled:bool=@n31
    Translation="0,-10,0"
    Rotation="0,0,0"
    Scale="1,1,1"
   ) -> (Pose) {
    alphaInputType=Bool
    rotationMode=BMM_Replace
    rotationSpace=BCS_BoneSpace
    translationMode=BMM_Additive
    translationSpace=BCS_BoneSpace
    alphaBoolBlend {blendOption=Linear}
    boneToModify {boneName=dyn_jacketB_l_MSTR}
    translation {y=-10}
   }
   n33 = "Transform (Modify) Bone - Bone: dyn_jacketC_r_MSTR"(ComponentPose=@n32, bAlphaBoolEnabled=@n31, Translation="0,-5,0", Rotation="0,0,0", Scale="1,1,1") {
    alphaInputType=Bool
    rotationMode=BMM_Replace
    rotationSpace=BCS_BoneSpace
    translationMode=BMM_Additive
    translationSpace=BCS_BoneSpace
    alphaBoolBlend {blendOption=Linear}
    boneToModify {boneName=dyn_jacketC_r_MSTR}
    translation {y=-5}
   }
   n34 = "Transform (Modify) Bone - Bone: dyn_jacketC_r_02"!(ComponentPose=@n33, bAlphaBoolEnabled:bool=@n31, Translation="0,-8,0", Rotation="0,0,0", Scale="1,1,1") -> (Pose) {
    alphaInputType=Bool
    rotationMode=BMM_Replace
    rotationSpace=BCS_BoneSpace
    translationMode=BMM_Additive
    translationSpace=BCS_BoneSpace
    alphaBoolBlend {blendOption=Linear}
    boneToModify {boneName=dyn_jacketC_r_02}
    translation {y=-8}
   }
   n35 = "Transform (Modify) Bone - Bone: dyn_jacketB_r_MSTR"!(ComponentPose=@n34, bAlphaBoolEnabled:bool=@n31, Translation="0,-3,-5", Rotation="10,-10,0", Scale="1,1,1") -> (Pose) {
    alphaInputType=Bool
    rotationMode=BMM_Replace
    rotationSpace=BCS_BoneSpace
    translationMode=BMM_Additive
    translationSpace=BCS_BoneSpace
    alphaBoolBlend {blendOption=Linear}
    boneToModify {boneName=dyn_jacketB_r_MSTR}
    rotation {pitch=10, yaw=-10}
    translation {y=-3, z=-5}
   }
   Save cached pose 'FEIdleCache'(Pose=Component To Local(ComponentPose=@n35)) {cacheName=FEIdleCache}
   n36 = Get accSclCape() -> (accSclCape:Vector)
   n37 = RigidBody!(
    ComponentPose=Local To Component(LocalPose=Use cached pose 'backBlingCache'())
    Alpha
    OverrideWorldGravity:Vector=Get Gravity Override()
    ExternalForce
    ComponentLinearAccScale:Vector=@n36
    ComponentLinearVelScale:Vector=@n36
   ) -> (Pose) {
    bClampLinearTranslationLimitToRefPose=true
    bOverrideWorldGravity=true
    overridePhysicsAsset=&/M_MED_AccentWall_jacket_backbling_Physics
    simulationSpace=BaseBoneSpace
    alphaBoolBlend {blendOption=Linear}
    baseBoneRef {boneName=spine_05}
    componentAppliedLinearAccClamp=<6000 6000 6000>
    componentLinearAccScale=<0.5 0.5 0.5>
    componentLinearVelScale=<0.5 0.5 0.5>
   }
   n38 = "Constraint - dyn_jacketSleeve_l_02 (PPP)"!(ComponentPose=@n37, Alpha=1, ConstraintWeights_0=0.2, ConstraintWeights_1:float=0.4, ConstraintWeights_2:float=0.4) -> (Pose) {
    alphaBoolBlend {blendOption=Linear}
    boneToModify {boneName=dyn_jacketSleeve_l_02}
    constraintSetup [{targetBone {boneName=spine_03}}, {targetBone {boneName=dyn_jacketA_l_02}}, {targetBone {boneName=dyn_jacketB_l_03}}]
    constraintWeights [1,1,1]
   }
   n39 = "Transform (Modify) Bone - Bone: dyn_jacketSleeve_l_MSTR"!(ComponentPose=@n38, Alpha:float=1, Translation="0,2,0", Rotation="0,-5,0", Scale="1,1,1") -> (Pose) {
    rotationMode=BMM_Additive
    rotationSpace=BCS_BoneSpace
    translationMode=BMM_Additive
    translationSpace=BCS_BoneSpace
    alphaBoolBlend {blendOption=Linear}
    boneToModify {boneName=dyn_jacketSleeve_l_MSTR}
    rotation {yaw=-5}
    translation {y=2}
   }
   Save cached pose 'RBANCache'(Pose=Blend Poses by bool(
     BlendPose_0=Component To Local(ComponentPose=@n39)
     BlendPose_1=Use cached pose 'backBlingCache'()
     bActiveValue=Get isWearingLongBackblingCape() -> (isWearingLongBackblingCape:bool)
    ) {blendType=Linear}) {cacheName=RBANCache}
   n40 = RigidBody!(
    ComponentPose=Local To Component(LocalPose=Use cached pose 'preSimCache'())
    Alpha
    OverrideWorldGravity=Get Gravity Override()
    ExternalForce
   ) -> (Pose) {
    bClampLinearTranslationLimitToRefPose=true
    bOverrideWorldGravity=true
    overridePhysicsAsset=&/M_MED_AccentWall_jacket_backbling_Crouching_Physics
    simulationSpace=BaseBoneSpace
    alphaBoolBlend {blendOption=Linear}
    baseBoneRef {boneName=spine_05}
    componentAppliedLinearAccClamp=<6000 6000 6000>
    componentLinearAccScale=<0.5 0.5 0.5>
    componentLinearVelScale=<0.5 0.5 0.5>
   }
   n41 = Anim Dynamics - Bone: dyn_jacketSleeve_l_03!(ComponentPose=@n40, Alpha=1, GravityOverride:Vector="0,0,-980", ExternalForce="0,0,0") -> (Pose) {
    angularDampingOverride=1
    angularSpringConstant=6000
    bAngularSpring=true
    bOverrideAngularDamping=true
    bUseGravityOverride=true
    gravityScale=0.35
    simulationSpace=RootRelative
    alphaBoolBlend {blendOption=Linear}
    boundBone {boneName=dyn_jacketSleeve_l_03}
    componentLinearAccScale=<1 1 1>
    componentLinearVelScale=<1 1 1>
    gravityOverride {z=-980}
    physicsBodyDefinitions [{
      boundBone {boneName=dyn_jacketSleeve_l_03}
      boxExtents=<20 30 30>
      constraintSetup {angularLimitsMax=<15 3 15>, angularLimitsMin=<-15 -3 -15>, angularTarget=<1 0 0>}
      localJointOffset=<4 0 0>
     }]
   }
   n42 = Anim Dynamics - Bone: dyn_jacketSleeve_l_04!(ComponentPose=@n41, Alpha=1, GravityOverride:Vector="0,0,-980", ExternalForce="0,0,0") -> (Pose) {
    angularDampingOverride=1
    angularSpringConstant=6000
    bAngularSpring=true
    bOverrideAngularDamping=true
    bUseGravityOverride=true
    gravityScale=0.35
    simulationSpace=RootRelative
    alphaBoolBlend {blendOption=Linear}
    boundBone {boneName=dyn_jacketSleeve_l_04}
    componentLinearAccScale=<1 1 1>
    componentLinearVelScale=<1 1 1>
    gravityOverride {z=-980}
    physicsBodyDefinitions [{
      boundBone {boneName=dyn_jacketSleeve_l_04}
      boxExtents=<20 30 30>
      constraintSetup {angularLimitsMax=<15 3 15>, angularLimitsMin=<-15 -3 -15>, angularTarget=<1 0 0>}
      localJointOffset=<4 0 0>
     }]
   }
   n43 = "Constraint - dyn_jacketSleeve_l_02 (PPP)"(ComponentPose=@n42, Alpha=1, ConstraintWeights_0=0.2, ConstraintWeights_1=0.4, ConstraintWeights_2=0.4) {
    alphaBoolBlend {blendOption=Linear}
    boneToModify {boneName=dyn_jacketSleeve_l_02}
    constraintSetup [{targetBone {boneName=spine_03}}, {targetBone {boneName=dyn_jacketA_l_02}}, {targetBone {boneName=dyn_jacketB_l_03}}]
    constraintWeights [1,1,1]
   }
   n44 = "Transform (Modify) Bone - Bone: dyn_jacketSleeve_l_MSTR"(ComponentPose=@n43, Alpha=1, Translation="0,3,0", Rotation="0,0,0", Scale="1,1,1") {
    rotationMode=BMM_Additive
    rotationSpace=BCS_BoneSpace
    translationMode=BMM_Additive
    translationSpace=BCS_BoneSpace
    alphaBoolBlend {blendOption=Linear}
    boneToModify {boneName=dyn_jacketSleeve_l_MSTR}
    translation {y=3}
   }
   n45 = "Transform (Modify) Bone - Bone: dyn_jacketSleeve_l_02"(ComponentPose=@n44, Alpha=1, Translation="0,2,0", Rotation="0,-15,0", Scale="1,1,1") {
    rotationMode=BMM_Additive
    rotationSpace=BCS_BoneSpace
    translationMode=BMM_Additive
    translationSpace=BCS_BoneSpace
    alphaBoolBlend {blendOption=Linear}
    boneToModify {boneName=dyn_jacketSleeve_l_02}
    rotation {yaw=-15}
    translation {y=2}
   }
   n46 = RigidBody!(
    ComponentPose=Local To Component(LocalPose=Use cached pose 'preSimCache'())
    Alpha=Get tactRBANAlpha() -> (tactRBANAlpha:double)
    OverrideWorldGravity=Get Gravity Override()
    ComponentLinearAccScale:Vector="0.5,0.5,0.5"
    ComponentLinearVelScale:Vector="0.5,0.5,0.5"
   ) -> (Pose) {
    bClampLinearTranslationLimitToRefPose=true
    bOverrideWorldGravity=true
    overridePhysicsAsset=&M_MED_AccentWall_jacket_backbling_Physics
    simulationSpace=BaseBoneSpace
    alphaBoolBlend {blendOption=Linear}
    baseBoneRef {boneName=spine_05}
    componentAppliedLinearAccClamp=<6000 6000 6000>
    componentLinearAccScale=<0.5 0.5 0.5>
    componentLinearVelScale=<0.5 0.5 0.5>
   }
   n47 = Get isPlayingCustomEmote()
   n48 = "Transform (Modify) Bone - Bone: dyn_jacketB_r_02"!(ComponentPose=@n46, bAlphaBoolEnabled:bool=@n47, Translation="0,0,0", Rotation="0,-20,0", Scale="1,1,1") -> (Pose) {
    alphaInputType=Bool
    rotationMode=BMM_Additive
    rotationSpace=BCS_BoneSpace
    alphaBoolBlend {blendOption=Linear}
    boneToModify {boneName=dyn_jacketB_r_02}
    rotation {yaw=-20}
   }
   n49 = "Transform (Modify) Bone - Bone: dyn_jacketA_r_03"!(ComponentPose=@n48, bAlphaBoolEnabled:bool=@n47, Translation="0,0,0", Rotation="0,-10,0", Scale="1,1,1") -> (Pose) {
    alphaInputType=Bool
    rotationMode=BMM_Additive
    rotationSpace=BCS_BoneSpace
    alphaBoolBlend {blendOption=Linear}
    boneToModify {boneName=dyn_jacketA_r_03}
    rotation {yaw=-10}
   }
   n50 = Anim Dynamics - Bone: dyn_jacketSleeve_l_03(ComponentPose=@n49, Alpha=1, GravityOverride="0,0,-980", ExternalForce="0,0,0") {
    angularDampingOverride=1
    angularSpringConstant=6000
    bAngularSpring=true
    bOverrideAngularDamping=true
    bUseGravityOverride=true
    gravityScale=0.35
    simulationSpace=RootRelative
    alphaBoolBlend {blendOption=Linear}
    boundBone {boneName=dyn_jacketSleeve_l_03}
    componentLinearAccScale=<1 1 1>
    componentLinearVelScale=<1 1 1>
    gravityOverride {z=-980}
    physicsBodyDefinitions [{
      boundBone {boneName=dyn_jacketSleeve_l_03}
      boxExtents=<20 30 30>
      constraintSetup {angularLimitsMax=<15 3 15>, angularLimitsMin=<-15 -3 -15>, angularTarget=<1 0 0>}
      localJointOffset=<4 0 0>
     }]
   }
   n51 = Anim Dynamics - Bone: dyn_jacketSleeve_l_04(ComponentPose=@n50, Alpha=1, GravityOverride="0,0,-980", ExternalForce="0,0,0") {
    angularDampingOverride=1
    angularSpringConstant=6000
    bAngularSpring=true
    bOverrideAngularDamping=true
    bUseGravityOverride=true
    gravityScale=0.35
    simulationSpace=RootRelative
    alphaBoolBlend {blendOption=Linear}
    boundBone {boneName=dyn_jacketSleeve_l_04}
    componentLinearAccScale=<1 1 1>
    componentLinearVelScale=<1 1 1>
    gravityOverride {z=-980}
    physicsBodyDefinitions [{
      boundBone {boneName=dyn_jacketSleeve_l_04}
      boxExtents=<20 30 30>
      constraintSetup {angularLimitsMax=<15 3 15>, angularLimitsMin=<-15 -3 -15>, angularTarget=<1 0 0>}
      localJointOffset=<4 0 0>
     }]
   }
   n52 = "Constraint - dyn_jacketSleeve_l_02 (PPP)"(ComponentPose=@n51, Alpha=1, ConstraintWeights_0=0.2, ConstraintWeights_1=0.4, ConstraintWeights_2=0.4) {
    alphaBoolBlend {blendOption=Linear}
    boneToModify {boneName=dyn_jacketSleeve_l_02}
    constraintSetup [{targetBone {boneName=spine_03}}, {targetBone {boneName=dyn_jacketA_l_02}}, {targetBone {boneName=dyn_jacketB_l_03}}]
    constraintWeights [1,1,1]
   }
   n53 = "Transform (Modify) Bone - Bone: dyn_jacketSleeve_l_MSTR"(ComponentPose=@n52, Alpha=1, Translation="0,2,0", Rotation="0,-3,0", Scale="1,1,1") {
    rotationMode=BMM_Additive
    rotationSpace=BCS_BoneSpace
    translationMode=BMM_Additive
    translationSpace=BCS_BoneSpace
    alphaBoolBlend {blendOption=Linear}
    boneToModify {boneName=dyn_jacketSleeve_l_MSTR}
    rotation {yaw=-3}
    translation {y=2}
   }
   n54 = "Transform (Modify) Bone - Bone: dyn_jacketSleeve_l_MSTR"!(ComponentPose=@n53, bAlphaBoolEnabled:bool=Get bIsSkyDiving(), Translation="0,0,0", Rotation="0,-10,0", Scale="1,1,1") -> (Pose) {
    alphaInputType=Bool
    rotationMode=BMM_Additive
    rotationSpace=BCS_BoneSpace
    alphaBoolBlend {blendOption=Linear}
    boneToModify {boneName=dyn_jacketSleeve_l_MSTR}
    rotation {yaw=-10}
   }
   n55 = Blend Poses by bool(BlendPose_0=Component To Local(ComponentPose=@n45), BlendPose_1=Component To Local(ComponentPose=@n54), bActiveValue=Get bIsCrouching()) {blendType=Linear}
   n56 = RigidBody!(
    ComponentPose=Local To Component(LocalPose=Use cached pose 'preSimCache'())
    Alpha
    OverrideWorldGravity=Get Gravity Override()
    ExternalForce:Vector
   ) -> (Pose) {
    bClampLinearTranslationLimitToRefPose=true
    bOverrideWorldGravity=true
    overridePhysicsAsset=&/M_MED_AccentWall_jacket_backbling_FrontEnd_Physics
    simulationSpace=BaseBoneSpace
    alphaBoolBlend {blendOption=Linear}
    baseBoneRef {boneName=spine_05}
    componentAppliedLinearAccClamp=<6000 6000 6000>
    componentLinearAccScale=<0.5 0.5 0.5>
    componentLinearVelScale=<0.5 0.5 0.5>
   }
   n57 = Anim Dynamics - Bone: dyn_jacketSleeve_l_03(ComponentPose=@n56, Alpha=1, GravityOverride="0,0,-980", ExternalForce="0,0,0") {
    angularDampingOverride=1
    angularSpringConstant=6000
    bAngularSpring=true
    bOverrideAngularDamping=true
    bUseGravityOverride=true
    gravityScale=0.35
    simulationSpace=RootRelative
    alphaBoolBlend {blendOption=Linear}
    boundBone {boneName=dyn_jacketSleeve_l_03}
    componentLinearAccScale=<1 1 1>
    componentLinearVelScale=<1 1 1>
    gravityOverride {z=-980}
    physicsBodyDefinitions [{
      boundBone {boneName=dyn_jacketSleeve_l_03}
      boxExtents=<20 30 30>
      constraintSetup {angularLimitsMax=<15 3 15>, angularLimitsMin=<-15 -3 -15>, angularTarget=<1 0 0>}
      localJointOffset=<4 0 0>
     }]
   }
   n58 = Anim Dynamics - Bone: dyn_jacketSleeve_l_04(ComponentPose=@n57, Alpha=1, GravityOverride="0,0,-980", ExternalForce="0,0,0") {
    angularDampingOverride=1
    angularSpringConstant=6000
    bAngularSpring=true
    bOverrideAngularDamping=true
    bUseGravityOverride=true
    gravityScale=0.35
    simulationSpace=RootRelative
    alphaBoolBlend {blendOption=Linear}
    boundBone {boneName=dyn_jacketSleeve_l_04}
    componentLinearAccScale=<1 1 1>
    componentLinearVelScale=<1 1 1>
    gravityOverride {z=-980}
    physicsBodyDefinitions [{
      boundBone {boneName=dyn_jacketSleeve_l_04}
      boxExtents=<20 30 30>
      constraintSetup {angularLimitsMax=<15 3 15>, angularLimitsMin=<-15 -3 -15>, angularTarget=<1 0 0>}
      localJointOffset=<4 0 0>
     }]
   }
   n59 = "Constraint - dyn_jacketSleeve_l_02 (PPP)"(ComponentPose=@n58, Alpha=1, ConstraintWeights_0=0.2, ConstraintWeights_1=0.4, ConstraintWeights_2=0.4) {
    alphaBoolBlend {blendOption=Linear}
    boneToModify {boneName=dyn_jacketSleeve_l_02}
    constraintSetup [{targetBone {boneName=spine_03}}, {targetBone {boneName=dyn_jacketA_l_02}}, {targetBone {boneName=dyn_jacketB_l_03}}]
    constraintWeights [1,1,1]
   }
   n60 = "Transform (Modify) Bone - Bone: dyn_jacketSleeve_l_MSTR"!(ComponentPose=@n59, Alpha:float=1, Translation="0,2,0", Rotation="0,0,0", Scale="1,1,1") -> (Pose) {
    rotationMode=BMM_Additive
    rotationSpace=BCS_BoneSpace
    translationMode=BMM_Additive
    translationSpace=BCS_BoneSpace
    alphaBoolBlend {blendOption=Linear}
    boneToModify {boneName=dyn_jacketSleeve_l_MSTR}
    translation {y=2}
   }
   n61 = "Transform (Modify) Bone - Bone: dyn_jacketSleeve_l_02"(ComponentPose=@n60, Alpha=1, Translation="0,0,0", Rotation="0,-10,0", Scale="1,1,1") {
    rotationMode=BMM_Additive
    rotationSpace=BCS_BoneSpace
    translationMode=BMM_Additive
    translationSpace=BCS_BoneSpace
    alphaBoolBlend {blendOption=Linear}
    boneToModify {boneName=dyn_jacketSleeve_l_02}
    rotation {yaw=-10}
   }
   n62 = "Transform (Modify) Bone - Bone: dyn_jacketSleeve_l_MSTR"!(ComponentPose=@n61, bAlphaBoolEnabled:bool=Get bIsSkyDiving(), Translation="0,5,0", Rotation="0,-10,0", Scale="1,1,1") -> (Pose) {
    alphaInputType=Bool
    rotationMode=BMM_Additive
    rotationSpace=BCS_BoneSpace
    alphaBoolBlend {blendOption=Linear}
    boneToModify {boneName=dyn_jacketSleeve_l_MSTR}
    rotation {yaw=-10}
    translation {y=5}
   }
   n63 = "Transform (Modify) Bone - Bone: dyn_collarLowB_r_MSTR"(ComponentPose=@n62, Alpha=1, Translation="0,0,0", Rotation="0,20,15", Scale="1,1,1") {
    rotationMode=BMM_Additive
    rotationSpace=BCS_BoneSpace
    translationMode=BMM_Additive
    translationSpace=BCS_BoneSpace
    alphaBoolBlend {blendOption=Linear}
    boneToModify {boneName=dyn_collarLowB_r_MSTR}
    rotation {roll=15, yaw=20}
   }
   n64 = Component To Local(ComponentPose=@n63)
   n65 = Get bIsInFrontEnd()
   n66 = RigidBody!(
    ComponentPose=Local To Component(LocalPose=Use cached pose 'preSimCache'())
    Alpha
    OverrideWorldGravity=Get Gravity Override()
    ExternalForce
    ComponentLinearAccScale:Vector="0.5,0.8,0.8"
    ComponentLinearVelScale:Vector="0.5,0.8,0.8"
   ) -> (Pose) {
    bClampLinearTranslationLimitToRefPose=true
    bOverrideWorldGravity=true
    overridePhysicsAsset=&/M_MED_AccentWall_jacket_Physics
    simulationSpace=BaseBoneSpace
    alphaBoolBlend {blendOption=Linear}
    baseBoneRef {boneName=spine_05}
    componentAppliedLinearAccClamp=<6000 6000 6000>
    componentLinearAccScale=<0.5 0.8 0.8>
    componentLinearVelScale=<0.5 0.8 0.8>
   }
   n67 = RigidBody!(
    ComponentPose=Local To Component(LocalPose=Use cached pose 'preSimCache'())
    Alpha
    OverrideWorldGravity=Get Gravity Override()
    ExternalForce
   ) -> (Pose) {
    bClampLinearTranslationLimitToRefPose=true
    bOverrideWorldGravity=true
    overridePhysicsAsset=&/M_MED_AccentWall_jacket_Crouching_Physics
    simulationSpace=BaseBoneSpace
    alphaBoolBlend {blendOption=Linear}
    baseBoneRef {boneName=spine_05}
    componentAppliedLinearAccClamp=<6000 6000 6000>
    componentLinearAccScale=<0.5 0.8 0.8>
    componentLinearVelScale=<0.5 0.8 0.8>
   }
   n68 = Blend Poses by bool(BlendPose_0=Component To Local(ComponentPose=@n67), BlendPose_1=Component To Local(ComponentPose=@n66), bActiveValue=Get bIsCrouching()) {blendType=Linear}
   n69 = RigidBody(ComponentPose=Local To Component(LocalPose=Use cached pose 'preSimCache'()), OverrideWorldGravity=Get Gravity Override()) {
    bClampLinearTranslationLimitToRefPose=true
    bOverrideWorldGravity=true
    overridePhysicsAsset=&/M_MED_AccentWall_jacket_FrontEnd_Physics
    simulationSpace=BaseBoneSpace
    alphaBoolBlend {blendOption=Linear}
    baseBoneRef {boneName=spine_05}
    componentAppliedLinearAccClamp=<6000 6000 6000>
    componentLinearAccScale=<0.6 0.5 0.5>
    componentLinearVelScale=<0.6 0.5 0.5>
   }
   n70 = Get isPlayingCustomEmote()
   n71 = "Transform (Modify) Bone - Bone: dyn_jacketB_r_02"(ComponentPose=@n69, bAlphaBoolEnabled=@n70, Translation="0,0,0", Rotation="0,-20,0", Scale="1,1,1") {
    alphaInputType=Bool
    rotationMode=BMM_Additive
    rotationSpace=BCS_BoneSpace
    alphaBoolBlend {blendOption=Linear}
    boneToModify {boneName=dyn_jacketB_r_02}
    rotation {yaw=-20}
   }
   n72 = "Transform (Modify) Bone - Bone: dyn_jacketA_l_02"!(ComponentPose=@n71, bAlphaBoolEnabled:bool=@n70, Translation="0,-2,0", Rotation="0,-5,0", Scale="1,1,1") -> (Pose) {
    alphaInputType=Bool
    rotationMode=BMM_Additive
    rotationSpace=BCS_BoneSpace
    translationMode=BMM_Additive
    translationSpace=BCS_BoneSpace
    alphaBoolBlend {blendOption=Linear}
    boneToModify {boneName=dyn_jacketA_l_02}
    rotation {yaw=-5}
    translation {y=-2}
   }
   n73 = "Transform (Modify) Bone - Bone: dyn_jacketA_r_03"(ComponentPose=@n72, bAlphaBoolEnabled=@n70, Translation="0,0,0", Rotation="0,-10,0", Scale="1,1,1") {
    alphaInputType=Bool
    rotationMode=BMM_Additive
    rotationSpace=BCS_BoneSpace
    alphaBoolBlend {blendOption=Linear}
    boneToModify {boneName=dyn_jacketA_r_03}
    rotation {yaw=-10}
   }
   n74 = "Constraint - dyn_jacketA_l_MSTR (T)"(ComponentPose=@n73, Alpha=0.5, ConstraintWeights_0=1) {
    alphaBoolBlend {blendOption=Linear}
    boneToModify {boneName=dyn_jacketA_l_MSTR}
    constraintSetup [{transformType=Translation, targetBone {boneName=deform_upperarm_bk_l}}]
    constraintWeights [1]
   }
   n75 = Component To Local(ComponentPose=@n74)
   n76 = Get bIsInFrontEnd()
   n77 = RigidBody!(
    ComponentPose=Local To Component(LocalPose=Blend Poses by bool(BlendPose_0=@n75, BlendPose_1=@n68, bActiveValue=@n76) {blendType=Linear})
    Alpha
    OverrideWorldGravity="0,0,0"
    ExternalForce
    ComponentLinearAccScale:Vector="0.6,0.6,0.6"
    ComponentLinearVelScale:Vector="0.6,0.6,0.6"
   ) -> (Pose) {
    bClampLinearTranslationLimitToRefPose=true
    bOverrideWorldGravity=true
    overridePhysicsAsset=&/M_MED_AccentWall_sleeve_Physics
    simulationSpace=BaseBoneSpace
    alphaBoolBlend {blendOption=Linear}
    baseBoneRef {boneName=spine_05}
    componentAppliedLinearAccClamp=<6000 6000 6000>
    componentLinearAccScale=<0.6 0.6 0.6>
    componentLinearVelScale=<0.6 0.6 0.6>
   }
   Save cached pose 'backBlingCache'(Pose=Blend Poses by bool(
     BlendPose_0=Blend Poses by bool(BlendPose_0=@n64, BlendPose_1=@n55, bActiveValue=@n65) {blendType=Linear}
     BlendPose_1=Component To Local(ComponentPose=@n77)
     bActiveValue=Get Is Wearing Backbling()
    ) {blendType=Linear}) {cacheName=backBlingCache}
   Use cached pose 'CRCache'()
  ]}
 EventGraph {nodes [
   n1 = Event BlueprintUpdateAnimation()
   > Set Delta Time X(Delta Time X=@n1.DeltaTimeX)
   > n2 = Sequence!(execute) -> (then_0, then_1, then_2:Exec, then_3:Exec, then_4:Exec, then_5:Exec, then_6:Exec, then_7:Exec)
   > Set bodyMesh(bodyMesh:Skeletal Mesh Component Object Reference=GetOwnerSkeletalMeshComponent()) -> (Output_Get:Skeletal Mesh Component Object Reference)
   n3 = Gravity Override State Function DEPRECATED(
    execute=@n2.then_1
    CharacterPartAnimInstance=Self-Reference()
    GameTimeInSeconds=Get Delta Time X()
    GravityOverrideParams=Get Gravity Override Params()
   )
   > Set Gravity Override(Gravity Override=@n3)
   n4 = Branch(execute=@n2.then_2, Condition=Get bIsCrouching())
   > n5 = Set isCrouchingAlpha(isCrouchingAlpha:double=1) -> (Output_Get:double)
   > Set isCrouchingAlpha(isCrouchingAlpha=FInterpTo(Current=@n5, Target=1, DeltaTime=0, InterpSpeed=0.3))
   n6 = Set isCrouchingAlpha(execute=@n4.else, isCrouchingAlpha=0)
   > Set isCrouchingAlpha(isCrouchingAlpha=FInterpTo(Current=@n6, Target=0, DeltaTime=0, InterpSpeed=0.3))
   n7 = Branch(execute=@n2.then_3, Condition=OR Boolean(A=Get bIsParachuteOpened(), B=Get bIsSkydiveFloating()))
   > Set isParachutFloating(isParachutFloating:bool=true) -> (Output_Get:bool)
   Set isParachutFloating(execute=@n7.else, isParachutFloating=false)
   n8 = Branch(execute=@n2.then_4, Condition=Get isWearingLongBackblingCape())
   > Set accSclCape(accSclCape:Vector="0.2,0.2,0.2") -> (Output_Get:Vector)
   Set accSclCape(execute=@n8.else, accSclCape="0.5,0.5,0.5")
   Branch(execute=@n2.then_5, Condition=Get bIsInFrontEnd())
   n9 = GetCurveValue(self=GetOwnerFrontEndAnimInstance(), CurveName=Frontend_Idle_Bypass)
   > Set CustomIdleIsPlayingFloat(CustomIdleIsPlayingFloat:float=@n9) -> (Output_Get:float)
   > n10 = Branch(Condition="Equal (Float)"(A=@n9))
   > Set notCustomIdlePlaying(notCustomIdlePlaying:bool=true) -> (Output_Get:bool)
   Set notCustomIdlePlaying(execute=@n10.else, notCustomIdlePlaying=false)
   n11 = GetAnimInstance(self=GetOwnerSkeletalMeshComponent())
   Is Valid(exec=@n2.then_6, InputObject=@n11)
   > n12 = Branch(Condition="Equal (Object)"(
     A=GetCurrentActiveMontage!(self:Anim Instance Object Reference=@n11) -> (ReturnValue:Anim Montage Object Reference)
     B=/BRCosmetics/Animation/Game/MainPlayer/Emotes/AccentWall/Hero/Emote_AccentWall_CMM_Hero_M
    ))
   > Set isPlayingCustomEmote(isPlayingCustomEmote:bool=true) -> (Output_Get:bool)
   Set isPlayingCustomEmote(execute=@n12.else, isPlayingCustomEmote=false)
   n13 = Branch(execute=@n2.then_7, Condition=Get bIsTacticalSprinting())
   > Set tactSprintAccVel(tactSprintAccVel:Vector="0, 0, 0") -> (Output_Get:Vector)
   > Set tactRBANAlpha(tactRBANAlpha:double=0.5) -> (Output_Get:double)
   Set tactSprintAccVel(execute=@n13.else, tactSprintAccVel="0.5,0.5,0.5")
   > Set tactRBANAlpha(tactRBANAlpha=1)
   Event OnCharacterCustomizationLoaded()
   n14 = CheckForBackbling()
   > Set Is Wearing Backbling(Is Wearing Backbling=@n14)
   > Set isWearingLongBackblingCape(isWearingLongBackblingCape:bool=OR Boolean(A=@n14.bIsBackblingACape, B=@n14.bIsBackblingLong)) -> (Output_Get:bool)
   AND Boolean!(A=@n14, B=NOT Boolean(A=@n14.bIsBackblingACape), C:bool=NOT Boolean(A=@n14.bIsBackblingLong)) -> (ReturnValue)
  ]}
]
variables [
 accSclCape {type=Vector}
 bodyMesh {type=SkeletalMeshComponent}
 CustomIdleIsPlayingFloat {type=number}
 Delta Time X {type=number}
 Gravity Override {type=Vector}
 Gravity Override Params {
  type=GravityOverrideParamsStruct
  default {skydive_GoingUp {gravityOverride=<0 -300 600>, windAmplitude=<200 -500 100>}, tactical_Sprint {gravityOverride=<0 0 0>}}
 }
 Is Wearing Backbling {type=boolean}
 isCrouchingAlpha {type=number}
 isParachutFloating {type=boolean}
 isPlayingCustomEmote {type=boolean}
 isWearingLongBackblingCape {type=boolean}
 NewVar {type=number}
 notCustomIdlePlaying {type=boolean}
 tactRBANAlpha {type=number}
 tactSprintAccVel {type=Vector}
]
